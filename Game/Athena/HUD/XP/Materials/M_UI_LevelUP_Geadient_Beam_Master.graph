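# Translucent UI master material for the XP HUD's level-up gradient beam.
# A sine of the UVs, masked by channel and offset by animated height, feeds a SmoothStep for soft edges.
# With HasFallOff, that mask combines with a rotated, power-faded linear gradient whose start animates.
# The result blends Color_A toward an animated hotspot color and, scaled by animated alpha, supplies opacity; Anim_ scalars drive each start-to-stop lerp.
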Material /Game/Athena/HUD/XP/Materials/M_UI_LevelUP_Geadient_Beam_Master
graph {
 properties {blendMode=BLEND_Translucent, materialDomain=MD_UI, pixelDepthOffsetMode=PDOM_Legacy, usedShadingModels=Default Lit}
 nodes [
  n1 = Sine(None=TextureCoordinate()) {period=2}
  n2 = LinearInterpolate(
   A=ScalarParameter() {defaultValue=1, parameterName=BeamHeight_Start}
   B=ScalarParameter() {parameterName=BeamHeight_Stop}
   Alpha=ScalarParameter() {parameterName=Anim_BeamHeight}
  )
  n3 = SmoothStep(
   Max=ScalarParameter() {defaultValue=0.125, parameterName=EdgeSoftness}
   Value=Subtract(A=StaticComponentMaskParameter(None=@n1) {defaultG=true, parameterName=EdgeSoftenessMaskDirection}, B=OneMinus(None=@n2)) {constB=0.25}
  )
  n4 = MaterialFunctionCall(
   GradientStart=LinearInterpolate(
    A=ScalarParameter() {defaultValue=0.5, parameterName=Gradient_Start}
    B=ScalarParameter() {parameterName=Gradient_Stop}
    Alpha=ScalarParameter() {defaultValue=1, parameterName=Anim_Grow}
   )
   RotationAngle=ScalarParameter() {defaultValue=-0.5, parameterName=Rotation}
  ) {materialFunction=&MF_UI_LinearGradient}
  n5 = ScalarParameter() {defaultValue=2, parameterName=BeamFadeRate}
  n6 = Multiply(A=Saturate(None=Power(Base=@n4.Result, Exp=@n5)), B=ScalarParameter() {defaultValue=1, parameterName=BeamAlpha})
  n7 = StaticSwitchParameter(True=DotProduct(A=@n3, B=NamedRerouteDeclaration(None=@n6) {name=Beam, nodeColor=<rgba 0 1 0.1094 1>}), False=@n3) {defaultValue=true, parameterName=HasFallOff}
  LinearInterpolate(
   A=VectorParameter() {parameterName=Color_A, defaultValue=<rgba 0 1 0 1>}.RGB
   B=LinearInterpolate(
    A=VectorParameter() {parameterName=Color_B_Start, defaultValue=<rgba 1 0 0 1>}.RGB
    B=VectorParameter() {parameterName=Color_B_Stop, defaultValue=<rgba 1 0 0 1>}.RGB
    Alpha=ScalarParameter() {parameterName=Anim_HotspotColor}
   )
   Alpha=@n7
  )
  Multiply(
   A=@n7
   B=LinearInterpolate(
    A=ScalarParameter() {defaultValue=1, parameterName=Alpha_Start}
    B=ScalarParameter() {defaultValue=1, parameterName=Alpha_Stop}
    Alpha=ScalarParameter() {defaultValue=1, parameterName=Anim_Alpha}
   )
  )
 ]
}
