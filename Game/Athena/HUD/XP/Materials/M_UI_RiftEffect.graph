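# Translucent UI material for a rift effect in the Athena XP HUD, built on a scrolling, noise-distorted two-channel mask.
# Panned Perlin noise offsets the UVs, which then pan with an adjustable vertical placement.
# Mask G blends main and accent colors, while doubled R is multiplied by a rerouted value.
# A rotated-UV horizontal gradient, shifted by an animatable start-to-end pip progress value, provides a progress mask.
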Material /Game/Athena/HUD/XP/Materials/M_UI_RiftEffect
graph {
 properties {bAutomaticallySetUsageInEditor=false, blendMode=BLEND_Translucent, materialDomain=MD_UI, usedShadingModels=Default Lit}
 nodes [
  n1 = TextureSampleParameter2D(UVs=NamedRerouteUsage()) {
   desc=Should have two channels R is Main mask and G is Accent
   parameterName=Mask
   texture=&/T_UI_Boon_HauntedSmoke_Rift
  }
  LinearInterpolate(
   A=VectorParameter() {parameterName=MainColor, defaultValue=<rgba 0.1045 0.7622 0.2954 1>}.RGB
   B=VectorParameter() {parameterName=AccentColor, defaultValue=<rgba 1 1 1 1>}.RGB
   Alpha=@n1.G
  )
  Multiply(A=Add(A=@n1.R, B=@n1.R), B=NamedRerouteUsage())
  n2 = TextureCoordinate()
  n3 = Constant() {r=0.5}
  n4 = ScalarParameter() {
   defaultValue=1
   desc="When progress value changes, it defines where change starts.\r\n0 - 1"
   group=Progress
   parameterName=PipProgressStart
   sliderMax=1
   sortPriority=1
  }
  n5 = ScalarParameter() {
   defaultValue=0.1
   desc="When progress value changes, it defines where change ends.\r\n0 - 1"
   group=Progress
   parameterName=PipProgressEnd
   sliderMax=1
   sortPriority=2
  }
  n6 = ScalarParameter() {
   defaultValue=1.3
   desc="Property for animating progress value change (from ProgressStart to ProgressEnd)\r\n0 - 1"
   group=Progress
   parameterName=AnimatePipProgress
   sliderMax=1
   sortPriority=3
  }
  n7 = Add(
   A=Multiply(A=MaterialFunctionCall(UVs=@n2, "Rotation Angle (0-1)"=@n3) {materialFunction=&CustomRotator}.Rotated Values)
   B=NamedRerouteDeclaration(None=LinearInterpolate(A=@n4, B=@n5, Alpha=@n6)) {name=Progress_1, nodeColor=<rgba 0 0 0 1>}
  ) {constB=0.5}
  NamedRerouteDeclaration(None=Saturate(None=ComponentMask(None=@n7) {r=true})) {name=ProgressGraident_Horizontal, nodeColor=<rgba 0 0 0 1>}
  n8 = ScalarParameter() {defaultValue=0.4091, group=Distortion, parameterName=DistortionSpeedX}
  n9 = ScalarParameter() {group=Distortion, parameterName=DistortionSpeedY}
  n10 = TextureSampleParameter2D(UVs=Panner(Coordinate=TextureCoordinate(), Speed=AppendVector(A=@n8, B=@n9)) {speedX=0.4}) {group=Distortion, parameterName=DistortionNoise, texture=&/T_UI_PerlinNoiseRift}
  n11 = AppendVector(A=@n10.G, B=@n10.G)
  n12 = ScalarParameter() {defaultValue=0.24, group=Distortion, parameterName=DistortionSize}
  n13 = Panner(
   Coordinate=Add(A=Multiply(A=@n11, B=@n12) {constB=0.24}, B=TextureCoordinate())
   Speed=AppendVector(A=ScalarParameter() {defaultValue=1, parameterName=SpeedX}, B=ScalarParameter() {parameterName=SpeedY})
  ) {speedX=1.1}
  n14 = ComponentMask(None=@n13) {g=true}
  n15 = ScalarParameter() {defaultValue=1, parameterName=PlacementOnVAxis}
  NamedRerouteDeclaration(None=AppendVector(A=ComponentMask(None=@n13) {r=true}, B=Add(A=@n14, B=@n15))) {name=UV's, nodeColor=<rgba 0 0 0 1>}
 ]
}
