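# M_Flare_Halo_Corona is an additive, unlit UI material for an animated halo corona flare in the Ashton LTM HUD.
# Two counter-rotating spikeball texture samples combine into boosted rays, scaled by Brightness.
# A Size-driven sphere mask looks up a halo gradient that colors the rays, modulated by multiplied time sines blended through a Flicker dynamic parameter.
# A static switch optionally applies an inverted exponential radial gradient mask.
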Material /Game/Athena/HUD/LTM/Ashton/Materials/M_Flare_Halo_Corona
graph {
 properties {
  blendMode=BLEND_Additive
  materialDomain=MD_UI
  pixelDepthOffsetMode=PDOM_Legacy
  shadingModel=MSM_Unlit
  twoSided=true
  usedShadingModels=Unlit
 }
 nodes [
  n1 = MaterialFunctionCall(
   UVs=MaterialFunctionCall(Texture Scale=VectorParameter() {group=Mask, parameterName=RadialSize, defaultValue=<rgba 1 1 0 1>}.RGB) {materialFunction=&ScaleUVsByCenter}.UVs
   Radius=ScalarParameter() {defaultValue=0.5, group=Mask, parameterName=MaskRadius}
   Density=ScalarParameter() {defaultValue=2.3, group=Mask, parameterName=MaskDensity}
  ) {materialFunction=&RadialGradientExponential}
  n2 = Multiply(A=Time())
  n3 = Sine(None=@n2) {period=0.075}
  n4 = Multiply(A=Multiply(A=Sine(None=@n2) {period=0.155}, B=Sine(None=@n2) {period=0.2}), B=Sine(None=@n2) {period=0.43})
  n5 = LinearInterpolate(
   A=Clamp(None=Multiply(A=@n3, B=@n4)) {maxDefault=3}
   Alpha=DynamicParameter() {defaultValue=<rgba 0.6 1 1 1>, paramNames [Size,RotateSpeed,Flicker,Param4]}.Flicker
  ) {constA=1}
  n6 = Constant() {r=1}
  n7 = TextureCoordinate() {uTiling=0.25, vTiling=0.25}
  n8 = Multiply(A=Time(), B=ScalarParameter() {defaultValue=1, parameterName=RotationSpeed})
  n9 = TextureSample(UVs=Rotator(Coordinate=Add(A=@n7) {constB=0.375}, Time=@n8)) {texture=/Game/Effects/Fort_Effects/Textures/Flares/T_Flare_Spikeball}
  n10 = TextureCoordinate() {uTiling=0.15, vTiling=0.15}
  n11 = TextureSample(UVs=Rotator(Coordinate=Add(A=@n10) {constB=0.425}, Time=@n8) {speed=-0.72}) {texture=&T_Flare_Spikeball}
  n12 = OneMinus(None=@n11.RGB)
  n13 = Multiply(
   A=Multiply(A=@n9.RGB, B=Power(Base=@n12) {constExponent=5})
   B=Multiply(A=ScalarParameter() {defaultValue=1, parameterName=Brightness}) {constB=7}
  ) {constB=80}
  n14 = Constant() {r=0.5}
  n15 = ScalarParameter() {defaultValue=1, parameterName=Size}
  n16 = TextureSample(UVs=SphereMask(A=TextureCoordinate(), B=@n14, Radius=@n14, Hardness=LinearInterpolate(Alpha=@n15) {constA=1, constB=0}) {hardnessPercent=25}) {texture=/Game/Effects/Fort_Effects/Textures/Grad/T_Halo_Corona_Gradient}
  n17 = Clamp(None=Multiply(A=Multiply(A=@n5, B=@n6), B=Multiply(A=@n13, B=@n16.RGB)))
  StaticSwitchParameter(True=Multiply(A=OneMinus(None=@n1.RadialGradientExponential), B=@n17), False=@n17) {parameterName=RadialMask}
  Subtract(A=@n9.RGB, B=@n11.RGB)
 ]
}
