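# Sound cue for the in-game UI sound that, going by its name, plays when an XP medal socket upgrades.
# A single wave player feeds one upgrade sample straight to the output, with boosted volume and a very large max distance.
# It routes through the critical in-game UI sound class and overrides concurrency to allow two instances, stopping the oldest.
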SoundCue /Game/Sounds/Fort_UI_Sounds/Gameplay/XP/Fort_XP_Medal_Socket_Upgrade_Cue
root=WavePlayer_0
nodes [{id=WavePlayer_0, properties {soundWaveAssetPtr=&/S11XP_Medal_Socket_Upgrade_01}}]
properties {
 bOverrideConcurrency=true
 maxDistance=2097152
 soundClassObject=&Fort_UI_InGame_Critical
 volumeMultiplier=1.5
 concurrencyOverrides {maxCount=2, resolutionRule=StopOldest}
}
