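# UIEmitter is a UMG particle emitter widget that spawns animated UIEmitterElement child widgets inside an overlay.
# StartEmitter fills up to MaxElements, randomizing each element's size, angle, distance, lifetime, spin and material.
# Negative padding offsets size and distance, with render rotation setting direction.
# Without RunOnce it binds each element's death event, while an optional retriggerable-delay lifetime sets EmitterExpired to block further spawning.
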WidgetBlueprint /Game/UI/Foundation/UIEmitter/UIEmitter
graphs [EventGraph {nodes [
   CreateEmitterElement() {is=title K2Node_CustomEvent}
   > n1 = Branch(Condition=Get EmitterExpired() -> (EmitterExpired:bool))
   n2 = Get Overlay_Emitter() -> (Overlay_Emitter:Overlay Object Reference)
   Branch(
    execute=@n1.else
    Condition="integer < integer"!(A:int=GetChildrenCount!(self:Panel Widget Object Reference=@n2) -> (ReturnValue:int=0), B:int=Get MaxElements() -> (MaxElements:int=0)) -> (ReturnValue:bool) {is=title K2Node_CallFunction}
   )
   > n3 = Create Widget!(
    execute:Exec
    Class:User Widget Class Reference=&/UIEmitterElement.UIEmitterElement_C
    OwningPlayer:Player Controller Object Reference=GetOwningPlayer()
   ) -> (then:Exec, ReturnValue:UIEmitter Element Object Reference)
   > n4 = Sequence()
   n5 = BreakVector2D(InVec=Get SizeRange() -> (SizeRange:Vector 2D Structure))
   n6 = RandomFloatInRange(Min=@n5, Max=@n5.Y)
   n7 = Get Image_EmitterMaterial!(self:UIEmitter Element Object Reference=@n3) -> (Image_EmitterMaterial:Image Object Reference)
   > SetDesiredSizeOverride!(execute, self:Image Object Reference=@n7, DesiredSize:Vector 2D Structure=MakeVector2D(X=@n6, Y=@n6)) -> (then)
   n8 = float / float(A=@n6, B=-2) {is=title K2Node_CallFunction}
   > SetPadding!(
    execute
    self:Overlay Slot Object Reference=SlotAsOverlaySlot(Widget=@n7)
    InPadding:Margin Structure=Make Margin(Left=@n8, Top=@n8, Right=@n8, Bottom=@n8)
   ) -> (then)
   n9 = BreakVector2D(InVec=Get AngleRange() -> (AngleRange:Vector 2D Structure))
   > SetRenderTransformAngle!(execute, self:Widget Object Reference=@n3, Angle:float=RandomFloatInRange(Min=@n9, Max=@n9.Y)) -> (then)
   n10 = BreakVector2D(InVec=Get DistanceRange() -> (DistanceRange:Vector 2D Structure))
   n11 = RandomFloatInRange(Min=@n10, Max=@n10.Y)
   > SetWidthOverride(
    self=Get SizeBox_Distance!(self:UIEmitter Element Object Reference=@n3) -> (SizeBox_Distance:Size Box Object Reference)
    InWidthOverride=@n11
   )
   > SetPadding(
    self=SlotAsOverlaySlot(Widget=Get CommonBorder_SizeContainer!(self:UIEmitter Element Object Reference=@n3) -> (CommonBorder_SizeContainer:Common Border Object Reference))
    InPadding=Make Margin(Left=0, Top=0, Right=int * float!(A:int=-1, B:double=@n11) -> (ReturnValue:double), Bottom=0)
   )
   > AddChildToOverlay!(execute, self:Overlay Object Reference=Get Overlay_Emitter(), Content:Widget Object Reference=@n3) -> (then, ReturnValue:Overlay Slot Object Reference)
   > SetImage!(execute, self, EmitterElement:UIEmitter Element Object Reference=@n3) -> (then)
   n12 = BreakVector2D(InVec=Get ParticleLifeRange() -> (ParticleLifeRange:Vector 2D Structure))
   > Set LifeDuration!(execute, LifeDuration:double=RandomFloatInRange(Min=@n12, Max=@n12.Y), self:UIEmitter Element Object Reference=@n3) -> (then, Output_Get:double)
   n13 = BreakVector2D(InVec=Get SpinSpeedRange() -> (SpinSpeedRange:Vector 2D Structure))
   > Set SpinSpeed!(execute, SpinSpeed:double=RandomFloatInRange(Min=@n13, Max=@n13.Y), self:UIEmitter Element Object Reference=@n3) -> (then, Output_Get:double)
   > StartAnim(self:UIEmitter Element Object Reference=@n3)
   n14 = Branch(execute=@n4.then_1, Condition=Get RunOnce() -> (RunOnce:bool))
   Bind Event to Emitter Death(execute=@n14.else, self:UIEmitter Element Object Reference=@n3, Delegate=Create Event()) {is=title K2Node_AddDelegate}
   n15 = SetImage!() -> (OutputDelegate, then, EmitterElement:UIEmitter Element Object Reference) {is=title K2Node_CustomEvent}
   n16 = Get Materials() -> (Materials:Array of Material Instance Object References)
   n17 = int - int(A=Length(TargetArray:Array of Material Instance Object References=@n16), B=1) {is=title K2Node_CallFunction}
   > SetBrushFromMaterial(
    self=Get Image_EmitterMaterial(self=@n15.EmitterElement)
    Material="Get (a copy)"(Array:Array of Material Instance Object References=@n16, Dimension 1=RandomIntegerInRange!(Min:int=0, Max:int=@n17) -> (ReturnValue:int=0)) -> (Output:Material Instance Object Reference)
   )
   StartEmitter() {is=title K2Node_CustomEvent}
   > Set EmitterExpired(EmitterExpired:bool=false) -> (Output_Get:bool)
   > n18 = For Loop(LastIndex=int - int(A=Get MaxElements() -> (MaxElements=0), B=1))
   SetVisibility(execute=@n18.Completed, InVisibility=HitTestInvisible)
   > Branch(Condition="Not Equal (Vector2D)"!(A:Vector 2D Structure=Get EmitterLifeRange() -> (EmitterLifeRange:Vector 2D Structure), B:Vector 2D Structure, ErrorTolerance:float=0.0001) -> (ReturnValue:bool))
   n19 = BreakVector2D(InVec=Get EmitterLifeRange())
   > RetriggerableDelay(Duration=RandomFloatInRange(Min=@n19, Max=@n19.Y))
   > Set EmitterExpired(EmitterExpired=true)
   CreateEmitterElement(execute=@n18.LoopBody) {is=title K2Node_CallFunction}
   Event PreConstruct()
   > SetImage(EmitterElement=Get UIEmmitterElement() -> (UIEmmitterElement:UIEmitter Element Object Reference)) {is=title K2Node_CallFunction}
   Event Construct()
   > ClearChildren(self=Get Overlay_Emitter())
   > n20 = Branch(Condition=Get StartAutomatically() -> (StartAutomatically:bool))
   > StartEmitter() {is=title K2Node_CallFunction}
   SetVisibility(execute=@n20.else, InVisibility=Collapsed)
  ]}]
variables [
 AngleRange {type=Vector2D, default=<0 360>}
 DistanceRange {type=Vector2D, default=<100 200>}
 EmitterExpired {type=boolean}
 EmitterLifeRange {type=Vector2D, default=<0 3>}
 Materials {type="array<MaterialInstance>", default [&/M_UI_DefaultEmitterMaterial_Inst]}
 MaxElements {type=integer, default=5}
 ParticleLifeRange {type=Vector2D, default=<1 3>}
 RunOnce {type=boolean}
 SizeRange {type=Vector2D, default=<16 64>}
 SpinSpeedRange {type=Vector2D, default=<0 0>}
 StartAutomatically {type=boolean, default=true}
]
widgets {
 info {inheritedWidgetCount=0, parentClass=&UserWidget, rootWidgetClass=&Overlay, widgetCount=2}
 widgets [
  Overlay_Emitter {widgetClassPath=&Overlay, bIsVariable=true, widget=Overlay_Emitter}
  UIEmmitterElement {
   widgetClassPath=&/UIEmitterElement
   bIsVariable=true
   parent=Overlay_Emitter
   widget=UIEmmitterElement
   slot {class=OverlaySlot, properties {horizontalAlignment=HAlign_Center, verticalAlignment=VAlign_Center}}
  }
 ]
}
